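; Behavior tree evaluator, apparently injected for AI wielding the TeaCake Ice Gauntlets, that gates use of the weapon's secondary action.
; On enter, it caches the controlled Fort pawn and records the game time as the moment the weapon became owned.
; Starting conditions pass when the current weapon has the configured tag and the scalable first-use delay has elapsed since that time.
; The parent's evaluation is called, but its result is ignored.
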(event EventOnEnter
  (params (OwnerComp :type Behavior%20Tree%20Component%20Object%20Reference))
  (bind _ascontroller (Utilities|Casting|CastToController (Components|GetOwner OwnerComp)))
  (bind _asfort_pawn (Utilities|Casting|CastToFortPawn (Pawn|GetControlledPawn _ascontroller)))
  (Variables|Default|SetFortPawn _asfort_pawn)
  (Variables|Default|SetTimeWhenWeaponOwned (Utilities|Time|GetGameTimeinSeconds)))

(fn EvaluateStartingConditions
  (params (OwnerComp :type Behavior%20Tree%20Component%20Object%20Reference))
  (|CallParentFunction OwnerComp)
  (return (and (GameplayTags|HasTag (Tags|GetWeaponDataGameplayTags (Variables|/Script/FortniteGame.FortPawn|GetCurrentWeapon (Variables|Default|GetFortPawn))) (Variables|Default|GetWeaponTagToCheck)) (>= (- (Utilities|Time|GetGameTimeinSeconds) (Variables|Default|GetTimeWhenWeaponOwned)) (ScalableFloat|GetValueatLevel (Variables|Default|GetSF_FirstUseDelay))))))
